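; AN_SteelKiteREsetIdle is an anim notify for the Steel Kite consumable, restarting its icy wisp trail effect when triggered from an animation.
; On notify, it casts the mesh component's owner to B_Consumable_SteelKite_Athena, then resets that actor's NS_IceTrail_FrostyWisps Niagara system.
; It returns false and does nothing else.
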(fn Received_Notify
  (params (MeshComp :type Skeletal%20Mesh%20Component%20Object%20Reference) (Animation :type Anim%20Sequence%20Base%20Object%20Reference) (EventReference :type Anim%20Notify%20Event%20Reference%20Structure%20%28by%20ref%29))
  (bind _asb_consumable_steel_kite_athena (Utilities|Casting|CastToB_Consumable_SteelKite_Athena (Components|GetOwner MeshComp)))
  (Niagara|ResetSystem (Variables|/WildEstateConsumables/Gameplay/SteelKite/B_Consumable_SteelKite_Athena.B_Consumable_SteelKite_Athena_C|GetNS_IceTrail_FrostyWisps _asb_consumable_steel_kite_athena))
  (return false))
