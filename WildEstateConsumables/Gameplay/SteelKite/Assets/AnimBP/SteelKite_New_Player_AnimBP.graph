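# SteelKite item animation Blueprint for the Fortnite player skeleton, supplying Item Override layers, of which only ItemFullBodyOverride has logic.
# It twists the pelvis by NW_Run_Alpha, layering spine and legs back from the cached input.
# When crouch-moving without sliding, Peely offsets bend five spine bones and shift the pelvis; otherwise only spine_05 is adjusted.
# A final mesh-space blend restores clavicles and hand IK from the input pose.
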AnimBlueprint /WildEstateConsumables/Gameplay/SteelKite/Assets/AnimBP/SteelKite_New_Player_AnimBP
target_skeleton=&Fortnite_M_Avg_Player_Skeleton
graphs [
 AnimGraph {nodes [Output Pose()]}
 EventGraph {links [], nodes []}
 ItemAimOffset {nodes [
   Input Pose!() -> (Pose, AimOffsetAlpha:double, Yaw:double, Pitch:double) {name=InPose_AimOffset}
   Input Pose!() -> (Pose) {name=InPose_PreAimOffsets}
   Output Pose() {layerGroup=Item Override}
  ]}
 ItemFallAdditive {nodes [
   Input Pose() {name=InPoseItemFallAdditive}
   Output Pose() {layerGroup=Item Override}
  ]}
 ItemFallLandAdditive {nodes [
   Input Pose() {name=InPoseItemFallLandAdditive}
   Output Pose() {layerGroup=Item Override}
  ]}
 ItemFallLevel2Override {nodes [
   Input Pose() {name=InPose_FallLevel2}
   Output Pose() {layerGroup=Item Override}
  ]}
 ItemFinalPoseOverride {nodes [
   Input Pose() {name=InFinalPose}
   Output Pose() {layerGroup=Item Override}
  ]}
 ItemFullBodyAimOffsets {nodes [
   Input Pose!() -> (Pose, FullBody_AimOffsetAlpha:double, Aim_Yaw:double, Aim_Pitch:double) {name=InPoseFullBodyAimOffsets}
   Input Pose!() -> (Pose) {name=InPosePreFullBodyAimOffsets}
   Output Pose() {layerGroup=Item Override}
  ]}
 ItemFullBodyOverride {nodes [
   n1 = Layered blend per bone(BasePose=Use cached pose 'PostSpineAdjustment'(), BlendPoses_0=Use cached pose 'FullBodyCache'()) {bMeshSpaceRotationBlend=true, layerSetup [{branchFilters [{boneName=clavicle_l}, {boneName=clavicle_r}, {boneName=ik_hand_root}]}]}
   Output Pose(Result=@n1) {layerGroup=Item Override}
   Save cached pose 'FullBodyCache'(Pose=Input Pose() {name=InPoseFullBody}) {cacheName=FullBodyCache}
   n2 = "Transform (Modify) Bone - Bone: pelvis"!(
    ComponentPose=Local To Component(LocalPose=Use cached pose 'FullBodyCache'())
    Alpha=Get NW_Run_Alpha() -> (NW_Run_Alpha:double)
    Rotation_Roll:float=0
    Rotation_Pitch:float=0
    Rotation_Yaw:float=Get PelvisTwist() -> (PelvisTwist:float)
   ) -> (Pose) {rotationMode=BMM_Additive, rotationSpace=BCS_ParentBoneSpace, alphaBoolBlend {blendOption=Linear}, boneToModify {boneName=pelvis}}
   n3 = Layered blend per bone(BasePose=Component To Local(ComponentPose=@n2), BlendPoses_0=Use cached pose 'FullBodyCache'()) {
    bMeshSpaceRotationBlend=true
    layerSetup [{branchFilters [{blendDepth=3, boneName=Spine_01}, {boneName=ik_foot_root}, {boneName=thigh_r}, {boneName=thigh_l}]}]
   }
   Save cached pose 'PreSpineAdjustment'(Pose=@n3) {cacheName=PreSpineAdjustment}
   n4 = Local To Component(LocalPose=Use cached pose 'PreSpineAdjustment'())
   n5 = Get PeelyOffsetCrouchMoving() -> (PeelyOffsetCrouchMoving:double)
   n6 = Get PeelyOffsetTwistCrouchMoving() -> (PeelyOffsetTwistCrouchMoving:double)
   n7 = "Transform (Modify) Bone - Bone: Spine_03"!(
    ComponentPose="Transform (Modify) Bone - Bone: Spine_02"!(
     ComponentPose="Transform (Modify) Bone - Bone: Spine_01"!(ComponentPose=@n4, Alpha=1, Translation="0,0,0", Rotation_Roll:float=@n6, Rotation_Pitch:float=0, Rotation_Yaw:float=@n5, Scale="1,1,1") -> (Pose) {rotationMode=BMM_Additive, rotationSpace=BCS_ParentBoneSpace, boneToModify {boneName=Spine_01}}
     Alpha=1
     Translation="0,0,0"
     Rotation_Roll:float=@n6
     Rotation_Pitch:float=0
     Rotation_Yaw:float=@n5
     Scale="1,1,1"
    ) -> (Pose) {rotationMode=BMM_Additive, rotationSpace=BCS_ParentBoneSpace, boneToModify {boneName=Spine_02}}
    Alpha=1
    Translation="0,0,0"
    Rotation_Roll:float=@n6
    Rotation_Pitch:float=0
    Rotation_Yaw:float=@n5
    Scale="1,1,1"
   ) -> (Pose) {rotationMode=BMM_Additive, rotationSpace=BCS_ParentBoneSpace, boneToModify {boneName=Spine_03}}
   n8 = "Transform (Modify) Bone - Bone: pelvis"!(
    ComponentPose="Transform (Modify) Bone - Bone: spine_05"!(
     ComponentPose="Transform (Modify) Bone - Bone: spine_04"!(ComponentPose=@n7, Alpha=1, Translation="0,0,0", Rotation_Roll:float=@n6, Rotation_Pitch:float=0, Rotation_Yaw:float=@n5, Scale="1,1,1") -> (Pose) {rotationMode=BMM_Additive, rotationSpace=BCS_ParentBoneSpace, boneToModify {boneName=spine_04}}
     Alpha=1
     Translation="0,0,0"
     Rotation_Roll:float=@n6
     Rotation_Pitch:float=0
     Rotation_Yaw:float=@n5
     Scale="1,1,1"
    ) -> (Pose) {rotationMode=BMM_Additive, rotationSpace=BCS_ParentBoneSpace, boneToModify {boneName=spine_05}}
    Alpha=1
    Translation_X:double=0
    Translation_Y:double=Get PeelyOffsetTransY() -> (PeelyOffsetTransY:double)
    Translation_Z:double=0
    Rotation_Roll=Get PeelyOffsetPelvisRoll() -> (PeelyOffsetPelvisRoll:double)
    Rotation_Pitch=0
    Rotation_Yaw=0
    Scale:Vector="1,1,1"
   ) -> (Pose) {
    rotationMode=BMM_Additive
    rotationSpace=BCS_ParentBoneSpace
    translationMode=BMM_Additive
    translationSpace=BCS_ParentBoneSpace
    boneToModify {boneName=pelvis}
   }
   n9 = "Transform (Modify) Bone - Bone: spine_05"(
    ComponentPose=Local To Component(LocalPose=Use cached pose 'PreSpineAdjustment'())
    Alpha=1
    Translation="0,0,0"
    Rotation_Roll=Get PeelyOffsetTwist() -> (PeelyOffsetTwist:float)
    Rotation_Pitch=0
    Rotation_Yaw=Get PeelyOffset() -> (PeelyOffset:float)
    Scale="1,1,1"
   ) {rotationMode=BMM_Additive, rotationSpace=BCS_ParentBoneSpace, boneToModify {boneName=spine_05}}
   Save cached pose 'PostSpineAdjustment'(Pose=Blend Poses by bool(
     BlendPose_0=Component To Local(ComponentPose=@n8)
     BlendPose_1=Component To Local(ComponentPose=@n9)
     bActiveValue=Get bIsCrouchMovingNotSliding() -> (bIsCrouchMovingNotSliding:bool)
    )) {cacheName=PostSpineAdjustment}
  ]}
 ItemIdleOverride {nodes [
   Input Pose() {name=InPose_Idle}
   Output Pose() {layerGroup=Item Override}
  ]}
 ItemJetPackJumpAdditive {nodes [Output Pose() {layerGroup=Item Override}]}
 ItemJetPackStartAdditive {nodes [Output Pose() {layerGroup=Item Override}]}
 ItemJumpLoopAdditive {nodes [
   Input Pose() {name=InPoseItemJumpLoopAdditive}
   Output Pose() {layerGroup=Item Override}
  ]}
 ItemJumpOverride {nodes [
   Input Pose() {name=InPoseJump}
   Output Pose() {layerGroup=Item Override}
  ]}
 ItemJumpUpAdditive {nodes [
   Input Pose() {name=InPoseItemJumpUpAdditive}
   Output Pose() {layerGroup=Item Override}
  ]}
 ItemPreAimOffsets {nodes [
   Input Pose() {name=InPosePreAimOffsets}
   Output Pose() {layerGroup=Item Override}
  ]}
 ItemPreIK {nodes [
   Input Pose() {name=InPose_PreIK}
   Output Pose() {layerGroup=Item Override}
  ]}
 ItemRightHandIK {nodes [
   Input Pose() {name=InPosePreRightHandIK}
   Input Pose() {name=InPosePostRightHandIK}
   Output Pose() {layerGroup=Item Override}
  ]}
 ItemSkeletalControl {nodes [
   Input Pose()
   Output Pose() {layerGroup=Item Override}
  ]}
 ItemSwimJumpFallLoop {nodes [Output Pose() {layerGroup=Item Override}]}
 ItemSwimJumpSurfaceEnd {nodes [Output Pose() {layerGroup=Item Override}]}
 ItemSwimJumpSurfaceLoop {nodes [Output Pose() {layerGroup=Item Override}]}
 ItemUpperBody {nodes [
   Input Pose!() -> (Pose, InputParam:Fort Anim Input Adjusted Aim Structure) {name=InPoseUpperBody}
   Output Pose() {layerGroup=Item Override}
  ]}
 ItemZipLineStartAdditive {nodes [Output Pose() {layerGroup=Item Override}]}
]
variables []
