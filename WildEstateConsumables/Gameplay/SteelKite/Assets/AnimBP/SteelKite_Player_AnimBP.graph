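# Player Anim Blueprint for the SteelKite consumable, layering additive twitches and a consume shudder onto the character through FullBodyOverride.
# A random timer, gated by isSlurpeeActive, triggers the Twitch Loops state machine, whose animations come from a Chooser.
# That result is cached as PostTwitch and feeds the ShudderAdd state machine, blended per bone into the output.
# Animation curves set ExitShudder and EndMontage, which drive shudder playback alongside a SteelKite gameplay tag check.
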AnimBlueprint /WildEstateConsumables/Gameplay/SteelKite/Assets/AnimBP/SteelKite_Player_AnimBP
target_skeleton=&Fortnite_M_Avg_Player_Skeleton
graphs [
 ANF_ResetTimer {nodes [
   ANF_ResetTimer!() -> (then, Context:"Anim Update Context Structure (by ref)", Node:"Anim Node Reference Structure (by ref)") {is=title K2Node_FunctionEntry}
   > Set TriggerTwitchLoop(TriggerTwitchLoop:bool=false) -> (Output_Get:bool)
   > Return Node()
  ]}
 AnimGraph {nodes [Output Pose()]}
 CharacterSkeletalControlPostLegIK {nodes [
   Input Pose() {name=InPosePostLegIK}
   Output Pose() {layerGroup=PlayerOverride}
  ]}
 EventGraph {nodes [
   TriggerSequenceOnce() {is=title K2Node_CustomEvent}
   > Branch(Condition=Get isSlurpeeActive() -> (isSlurpeeActive:bool))
   > Do Once()
   > Set TriggerTwitchLoop(TriggerTwitchLoop=true)
   > StartRandomTimer()
   Event BlueprintUpdateAnimation()
   > n1 = Sequence!(execute) -> (then_0, then_1, then_2:Exec, then_3:Exec)
   > Get!(execute:Exec) -> (then:Exec, else:Exec, Debug:bool)
   > PrintString!(
    execute
    InString="To String (Boolean)"(InBool=Get TriggerTwitchLoop() -> (TriggerTwitchLoop:bool))
    bPrintToScreen
    bPrintToLog
    TextColor
    Duration
    Key
   ) -> (then)
   n2 = Branch(execute=@n1.then_2, Condition=Get EndMontage() -> (EndMontage:bool))
   > Do Once(Reset=@n3.then)
   > Set PlayShudder(PlayShudder:bool=true) -> (Output_Get:bool)
   Branch(execute=@n2.else, Condition=Get ExitShudder() -> (ExitShudder:bool))
   > n3 = Set PlayShudder(PlayShudder=false)
   n4 = Cast To FortPlayerPawn(execute=@n1.then_3, Object=TryGetFortPawn())
   > Branch(Condition=HasMatchingGameplayTag!(self:Gameplay Tag Asset Interface Interface=@n4, TagToCheck:Gameplay Tag Structure=(TagName="Granted.SteelKite.Active")) -> (ReturnValue:bool))
   > n5 = Branch(Condition=Get EndMontage())
   Do Once(execute=@n5.else, Reset=@n5.then)
   > Set PlayShudder(PlayShudder=true)
   Set ExitShudder(execute=@n1.then_1, ExitShudder:bool="float > float"(A=GetCurveValue(CurveName=UserCurve01), B=0.1)) -> (Output_Get:bool)
   > Set EndMontage(EndMontage:bool="float > float"(A=GetCurveValue(CurveName=UserCurve02), B=0.1)) -> (Output_Get:bool)
   Event BlueprintInitializeAnimation()
   > StartRandomTimer()
  ]}
 FullBodyOverride {nodes [
   AnimGraphNode_StateMachine_1 = ShudderAdd() {is=title AnimGraphNode_StateMachine}
   n1 = Layered blend per bone(BasePose=Use cached pose 'Input pose'(), BlendPoses_0=@AnimGraphNode_StateMachine_1) {layerSetup [{branchFilters [
       {boneName=Root}
       {blendDepth=-1, boneName=ik_hand_l}
       {blendDepth=-1, boneName=ik_hand_r}
       {blendDepth=-1, boneName=ik_hand_root}
       {blendDepth=-1, boneName=ik_hand_gun}
      ]}]}
   Output Pose(Result=@n1) {layerGroup=PlayerOverride}
   Save cached pose 'Input pose'(Pose=Input Pose() {name=PassThroughFullBodyPose}) {cacheName=Input pose}
   AnimGraphNode_StateMachine_0 = Twitch Loops() {is=title AnimGraphNode_StateMachine}
   n2 = Layered blend per bone(
    BasePose=Use cached pose 'Input pose'()
    BlendPoses_0=Apply Additive!(Base=Use cached pose 'Input pose'(), Additive=@AnimGraphNode_StateMachine_0, bAlphaBoolEnabled:bool=Get isSlurpeeActive()) -> (Pose) {alphaInputType=Bool}
   ) {layerSetup [{branchFilters [
       {boneName=Root}
       {blendDepth=-1, boneName=ik_hand_l}
       {blendDepth=-1, boneName=ik_hand_r}
       {blendDepth=-1, boneName=ik_hand_root}
       {blendDepth=-1, boneName=ik_hand_gun}
      ]}]}
   Save cached pose 'PostTwitch'(Pose=@n2) {cacheName=PostTwitch}
  ]}
 Twitch Loops {
  entry=Pass Through
  path=FullBodyOverride.AnimGraphNode_StateMachine_0.Twitch Loops
  ids=id
  nodes [
   Alias() {is=title AnimStateAliasNode, also {aliases [Pass Through, Twitches]}}
   AnimStateNode_0 = Pass Through() {is=title AnimStateNode}
   AnimStateNode_1 = Twitches() {is=title AnimStateNode}
   AnimStateNode_2 = State() {is=title AnimStateNode}
   AnimStateTransitionNode_0 = AnimStateTransitionNode() {bAllowInertializationForSelfTransitions=true, blendMode=QuadraticInOut, crossfadeDuration=0}
   AnimStateTransitionNode_1 = AnimStateTransitionNode() {bAllowInertializationForSelfTransitions=true, bAutomaticRuleBasedOnSequencePlayerInState=true, crossfadeDuration=0}
   AnimStateTransitionNode_2 = AnimStateTransitionNode() {bAllowInertializationForSelfTransitions=true}
   AnimStateTransitionNode_3 = AnimStateTransitionNode() {bAllowInertializationForSelfTransitions=true, bDisabled=true}
  ]
  links [
   {
    dst=State
    rule=FullBodyOverride.AnimGraphNode_StateMachine_0.Twitch Loops.AnimStateTransitionNode_3.Transition
    src=Alias
    via=AnimStateTransitionNode_3
   }
   {
    dst=Twitches
    rule=FullBodyOverride.AnimGraphNode_StateMachine_0.Twitch Loops.AnimStateTransitionNode_0.Transition
    src=Pass Through
    via=AnimStateTransitionNode_0
   }
   {
    dst=Pass Through
    rule=FullBodyOverride.AnimGraphNode_StateMachine_0.Twitch Loops.AnimStateTransitionNode_2.Transition
    src=State
    via=AnimStateTransitionNode_2
   }
   {
    dst=Pass Through
    rule=FullBodyOverride.AnimGraphNode_StateMachine_0.Twitch Loops.AnimStateTransitionNode_1.Transition
    src=Twitches
    via=AnimStateTransitionNode_1
   }
  ]
 }
 Pass Through {
  path=FullBodyOverride.AnimGraphNode_StateMachine_0.Twitch Loops.AnimStateNode_0.Pass Through
  nodes [Output Animation Pose(Result=Additive Identity Pose())]
 }
 Twitches {
  path=FullBodyOverride.AnimGraphNode_StateMachine_0.Twitch Loops.AnimStateNode_1.Twitches
  nodes [
   AnimGraphNode_ChooserPlayer_0 = Chooser Player: CHT_Player_Twitches!() -> (Pose:Pose Link Structure) {
    is=title AnimGraphNode_ChooserPlayer
    chooser {_structType=/Script/Chooser.EvaluateChooser, chooser=&/CHT_Player_Twitches}
    chooserContextDefinition [
     {
      _structType=/Script/Chooser.ContextObjectTypeClass
      class=&/SteelKite_Player_AnimBP.SteelKite_Player_AnimBP_C
      direction=ReadWrite
     }
     {struct=/Script/Chooser.ChooserPlayerSettings, _structType=/Script/Chooser.ContextObjectTypeStruct, direction=Read}
    ]
    defaultSettings {blendOption=CubicInOut, blendTime=0.6667}
   }
   Output Animation Pose(Result=@AnimGraphNode_ChooserPlayer_0) {
    node {stateExitFunction {functionName=OnStateExit}}
    stateExitFunction {bSelfContext=true, memberGuid=#1, memberName=ANF_ResetTimer}
   }
  ]
 }
 AnimationBlendStackGraph_0 {
  path=FullBodyOverride.AnimGraphNode_StateMachine_0.Twitch Loops.AnimStateNode_1.Twitches.AnimGraphNode_ChooserPlayer_0.AnimationBlendStackGraph_0
  nodes [Output Pose!(Result:Pose Link Structure=Blend Stack Input!() -> (Pose:Pose Link Structure) {is=title AnimGraphNode_BlendStackInput}) {is=title AnimGraphNode_BlendStackResult}]
 }
 State {
  path=FullBodyOverride.AnimGraphNode_StateMachine_0.Twitch Loops.AnimStateNode_2.State
  nodes [
   Output Animation Pose(Result=Apply Additive!(
     Base=Additive Identity Pose()
     Additive="Sequence Player (additive) 'WildEstate_SteelKite_Consume_Shudder'"() {bLoopAnimation=false, sequence=/WildEstateConsumables/Gameplay/SteelKite/Assets/Anim/WildEstate_SteelKite_Consume_Shudder}
     Alpha:float
    ) -> (Pose))
   Use cached pose 'Input pose'()
  ]
 }
 Transition {
  path=FullBodyOverride.AnimGraphNode_StateMachine_0.Twitch Loops.AnimStateTransitionNode_0.Transition
  nodes [Result(bCanEnterTransition=Get TriggerTwitchLoop())]
 }
 Transition {path=FullBodyOverride.AnimGraphNode_StateMachine_0.Twitch Loops.AnimStateTransitionNode_1.Transition, nodes [Result()]}
 Transition {
  path=FullBodyOverride.AnimGraphNode_StateMachine_0.Twitch Loops.AnimStateTransitionNode_2.Transition
  nodes [Result(bCanEnterTransition=Get ExitShudder())]
 }
 Transition {
  path=FullBodyOverride.AnimGraphNode_StateMachine_0.Twitch Loops.AnimStateTransitionNode_3.Transition
  nodes [Result(bCanEnterTransition=Get EndMontage())]
 }
 ShudderAdd {
  entry=PlayShudder
  path=FullBodyOverride.AnimGraphNode_StateMachine_1.ShudderAdd
  ids=id
  nodes [
   AnimStateNode_0 = State()
   AnimStateNode_1 = PlayShudder() {is=title AnimStateNode}
   AnimStateTransitionNode_0 = AnimStateTransitionNode() {bAllowInertializationForSelfTransitions=true}
   AnimStateTransitionNode_1 = AnimStateTransitionNode() {bAllowInertializationForSelfTransitions=true}
  ]
  links [
   {
    dst=State
    rule=FullBodyOverride.AnimGraphNode_StateMachine_1.ShudderAdd.AnimStateTransitionNode_1.Transition
    src=PlayShudder
    via=AnimStateTransitionNode_1
   }
   {
    dst=PlayShudder
    rule=FullBodyOverride.AnimGraphNode_StateMachine_1.ShudderAdd.AnimStateTransitionNode_0.Transition
    src=State
    via=AnimStateTransitionNode_0
   }
  ]
 }
 State {
  path=FullBodyOverride.AnimGraphNode_StateMachine_1.ShudderAdd.AnimStateNode_0.State
  nodes [Output Animation Pose(Result=Use cached pose 'PostTwitch'())]
 }
 PlayShudder {
  path=FullBodyOverride.AnimGraphNode_StateMachine_1.ShudderAdd.AnimStateNode_1.PlayShudder
  nodes [
   Output Animation Pose(Result=Apply Additive(
     Base=Use cached pose 'Input pose'()
     Additive="Sequence Player (additive) 'WildEstate_SteelKite_Consume_Shudder'"() {bLoopAnimation=false, sequence=&WildEstate_SteelKite_Consume_Shudder}
    ))
   Use cached pose 'PostTwitch'()
  ]
 }
 Transition {
  path=FullBodyOverride.AnimGraphNode_StateMachine_1.ShudderAdd.AnimStateTransitionNode_0.Transition
  nodes [Result(bCanEnterTransition=Get PlayShudder() -> (PlayShudder:bool))]
 }
 Transition {
  path=FullBodyOverride.AnimGraphNode_StateMachine_1.ShudderAdd.AnimStateTransitionNode_1.Transition
  nodes [Result(bCanEnterTransition=Get ExitShudder())]
 }
 ItemAimOffset {nodes [
   Input Pose!() -> (Pose, AimOffsetAlpha:double, Yaw:double, Pitch:double) {name=InPose_AimOffset}
   Input Pose!() -> (Pose) {name=InPose_PreAimOffsets}
   Output Pose() {is=title AnimGraphNode_Root, layerGroup=Item Override}
  ]}
 ItemFallAdditive {nodes [
   Input Pose() {name=InPoseItemFallAdditive}
   Output Pose() {layerGroup=Item Override}
  ]}
 ItemFallLandAdditive {nodes [
   Input Pose() {name=InPoseItemFallLandAdditive}
   Output Pose() {layerGroup=Item Override}
  ]}
 ItemFallLevel2Override {nodes [
   Input Pose() {name=InPose_FallLevel2}
   Output Pose() {layerGroup=Item Override}
  ]}
 ItemFinalPoseOverride {nodes [
   Input Pose() {name=InFinalPose}
   Output Pose() {layerGroup=Item Override}
  ]}
 ItemFullBodyAimOffsets {nodes [
   Input Pose!() -> (Pose, FullBody_AimOffsetAlpha:double, Aim_Yaw:double, Aim_Pitch:double) {name=InPoseFullBodyAimOffsets}
   Input Pose!() -> (Pose) {name=InPosePreFullBodyAimOffsets}
   Output Pose() {layerGroup=Item Override}
  ]}
 ItemFullBodyOverride {nodes [
   Input Pose() {name=InPoseFullBody}
   Output Pose() {layerGroup=Item Override}
  ]}
 ItemIdleOverride {nodes [
   Input Pose() {name=InPose_Idle}
   Output Pose() {layerGroup=Item Override}
  ]}
 ItemJetPackJumpAdditive {nodes [Output Pose() {layerGroup=Item Override}]}
 ItemJetPackStartAdditive {nodes [Output Pose() {layerGroup=Item Override}]}
 ItemJumpLoopAdditive {nodes [
   Input Pose() {name=InPoseItemJumpLoopAdditive}
   Output Pose() {layerGroup=Item Override}
  ]}
 ItemJumpOverride {nodes [
   Input Pose() {name=InPoseJump}
   Output Pose() {layerGroup=Item Override}
  ]}
 ItemJumpUpAdditive {nodes [
   Input Pose() {name=InPoseItemJumpUpAdditive}
   Output Pose() {layerGroup=Item Override}
  ]}
 ItemPreAimOffsets {nodes [
   Input Pose() {name=InPosePreAimOffsets}
   Output Pose() {layerGroup=Item Override}
  ]}
 ItemPreIK {nodes [
   Input Pose() {name=InPose_PreIK}
   Output Pose() {layerGroup=Item Override}
  ]}
 ItemRightHandIK {nodes [
   Input Pose() {name=InPosePreRightHandIK}
   Input Pose() {name=InPosePostRightHandIK}
   Output Pose() {layerGroup=Item Override}
  ]}
 ItemSkeletalControl {nodes [
   Input Pose()
   Output Pose() {layerGroup=Item Override}
  ]}
 ItemSwimJumpFallLoop {nodes [Output Pose() {layerGroup=Item Override}]}
 ItemSwimJumpSurfaceEnd {nodes [Output Pose() {layerGroup=Item Override}]}
 ItemSwimJumpSurfaceLoop {nodes [Output Pose() {layerGroup=Item Override}]}
 ItemUpperBody {nodes [
   Input Pose!() -> (Pose, InputParam:Fort Anim Input Adjusted Aim Structure) {name=InPoseUpperBody}
   Output Pose() {layerGroup=Item Override}
  ]}
 ItemZipLineStartAdditive {nodes [Output Pose() {layerGroup=Item Override}]}
 PreLegIK {nodes [
   Input Pose!() -> (Pose)
   Output Pose() {layerGroup=PlayerOverride}
  ]}
 StartRandomTimer {nodes [
   StartRandomTimer() {is=title K2Node_FunctionEntry}
   > Set Timer by Function Name!(
    execute
    Object:Object Reference
    FunctionName:String=TriggerSequenceOnce
    Time:float=RandomFloatInRange(Min=1, Max=7)
    bLooping:bool=false
    bMaxOncePerFrame:bool=false
    InitialStartDelay:float=0
    InitialStartDelayVariance:float=0
   ) -> (then, ReturnValue:Timer Handle Structure) {is=title K2Node_CallFunction}
  ]}
 UpperBody {nodes [
   Input Pose() {name=InPoseUpperBody}
   Output Pose() {layerGroup=PlayerOverride}
  ]}
]
variables [
 Debug {type=boolean}
 EndMontage {type=boolean}
 ExitShudder {type=boolean}
 isSlurpeeActive {type=boolean}
 PlayShudder {type=boolean}
 TriggerTwitchLoop {type=boolean}
]
