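# Temporary SoundCue for SteelKite alert-idle voice chatter.
# A Random node plays a single temp wave, excluded from branch culling, so there is no variation yet.
# Output uses shared NPC alerted VO attenuation, the critical enemy VO sound class, a reduced volume and a SteelKite chatter concurrency set to limit overlapping lines.
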SoundCue /WildEstateConsumables/Gameplay/SteelKite/Assets/Audio/VO_SteelKite_AlertIdle_cue_TEMP
root=Random_0
nodes [
 {id=Random_0, children [WavePlayer_0], properties {bShouldExcludeFromBranchCulling=true, weights [1]}}
 {id=WavePlayer_0, properties {soundWaveAssetPtr=&/VO_SteelKite_AlertIdle_TEMP}}
]
properties {
 attenuationSettings=/Game/Sounds/Fort_Human_VO_BR/NPCs/Shared/NPC_VO_Alerted_Att
 maxDistance=12500
 soundClassObject=&Fort_VO_Enemy_Critical
 volumeMultiplier=0.7
 concurrencySet [&/SteelKite_Chatter_Conc]
}
